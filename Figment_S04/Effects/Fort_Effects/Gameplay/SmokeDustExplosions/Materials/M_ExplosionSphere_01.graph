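# Additive unlit material for mesh particles, named for an explosion sphere effect.
# Particle color is multiplied by a Fresnel function with tunable brightness and exponent.
# A depth fade with an adjustable distance softens intersections with scene geometry on Default, Medium and High quality, while Low uses the Fresnel color directly to skip that step.
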Material /Figment_S04/Effects/Fort_Effects/Gameplay/SmokeDustExplosions/Materials/M_ExplosionSphere_01
graph {
 properties {bUsedWithMeshParticles=true, blendMode=BLEND_Additive, pixelDepthOffsetMode=PDOM_Legacy, shadingModel=MSM_Unlit, usedShadingModels=Unlit}
 nodes [
  n1 = Multiply(
   A=ParticleColor().RGB
   B=MaterialFunctionCall(
    Brightness=ScalarParameter() {defaultValue=1, parameterName=Fresnel Brightness}
    Exponent=ScalarParameter() {defaultValue=10, parameterName=Fresnel Exponent}
   ) {materialFunction=&MF_Fresnel_FX}.Fresnel
  )
  n2 = DepthFade(Opacity=@n1, FadeDistance=ScalarParameter() {defaultValue=500, parameterName=Depth Fade Distance})
  QualitySwitch(Default=@n2, Low=@n1, High=@n2, Medium=@n2)
 ]
}
